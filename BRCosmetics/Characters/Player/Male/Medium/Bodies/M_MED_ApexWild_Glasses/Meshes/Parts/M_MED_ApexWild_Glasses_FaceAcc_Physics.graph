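# Physics asset for the face accessory mesh of the male medium ApexWild Glasses character, covering four hair strands (a to d) and the head.
# Each strand is a chain of three capsule bodies, constrained bone to parent bone, with the first link attached to a small head capsule.
# All swing and twist axes are Free, so the stored 45-degree limits do not restrict rotation.
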PhysicsAsset /BRCosmetics/Characters/Player/Male/Medium/Bodies/M_MED_ApexWild_Glasses/Meshes/Parts/M_MED_ApexWild_Glasses_FaceAcc_Physics
bodies {
 hair_a_01 [{shapeType=Capsule, length=3, radius=2, shapeName=hair_a_01_capsule, rotation=<rot 90 169.5213 159.5213>}]
 hair_a_02 [{shapeType=Capsule, length=3, radius=2, shapeName=hair_a_02_capsule, rotation=<rot 90 169.5213 159.5213>}]
 hair_a_03 [{shapeType=Capsule, length=3, radius=2, shapeName=hair_a_03_capsule, rotation=<rot 90 169.5213 159.5213>}]
 hair_b_01 [{shapeType=Capsule, length=3, radius=2.5, shapeName=hair_b_01_capsule, rotation=<rot 90 0 0>}]
 hair_b_02 [{shapeType=Capsule, length=3, radius=2.5, shapeName=hair_b_02_capsule, rotation=<rot 90 0 0>}]
 hair_b_03 [{shapeType=Capsule, length=3, radius=2.5, shapeName=hair_b_03_capsule, rotation=<rot 90 0 0>}]
 hair_c_01 [{shapeType=Capsule, length=3, radius=2, shapeName=hair_c_01_capsule}]
 hair_c_02 [{shapeType=Capsule, length=3, radius=2, shapeName=hair_c_02_capsule}]
 hair_c_03 [{shapeType=Capsule, length=3, radius=2, shapeName=hair_c_03_capsule}]
 hair_d_01 [{shapeType=Capsule, length=3, radius=2, shapeName=hair_d_01_capsule}]
 hair_d_02 [{shapeType=Capsule, length=3, radius=2, shapeName=hair_d_02_capsule}]
 hair_d_03 [{shapeType=Capsule, length=3, radius=2, shapeName=hair_d_03_capsule}]
 head [{shapeType=Capsule, length=1, radius=1}]
}
constraints [
 {
  bone1Name=hair_a_01
  bone2Name=head
  swing1LimitDegrees=45
  swing1Motion=Free
  swing2LimitDegrees=45
  swing2Motion=Free
  twistLimitDegrees=45
  twistMotion=Free
 }
 {
  bone1Name=hair_a_02
  bone2Name=hair_a_01
  swing1LimitDegrees=45
  swing1Motion=Free
  swing2LimitDegrees=45
  swing2Motion=Free
  twistLimitDegrees=45
  twistMotion=Free
 }
 {
  bone1Name=hair_a_03
  bone2Name=hair_a_02
  swing1LimitDegrees=45
  swing1Motion=Free
  swing2LimitDegrees=45
  swing2Motion=Free
  twistLimitDegrees=45
  twistMotion=Free
 }
 {
  bone1Name=hair_b_01
  bone2Name=head
  swing1LimitDegrees=45
  swing1Motion=Free
  swing2LimitDegrees=45
  swing2Motion=Free
  twistLimitDegrees=45
  twistMotion=Free
 }
 {
  bone1Name=hair_b_02
  bone2Name=hair_b_01
  swing1LimitDegrees=45
  swing1Motion=Free
  swing2LimitDegrees=45
  swing2Motion=Free
  twistLimitDegrees=45
  twistMotion=Free
 }
 {
  bone1Name=hair_b_03
  bone2Name=hair_b_02
  swing1LimitDegrees=45
  swing1Motion=Free
  swing2LimitDegrees=45
  swing2Motion=Free
  twistLimitDegrees=45
  twistMotion=Free
 }
 {
  bone1Name=hair_c_01
  bone2Name=head
  swing1LimitDegrees=45
  swing1Motion=Free
  swing2LimitDegrees=45
  swing2Motion=Free
  twistLimitDegrees=45
  twistMotion=Free
 }
 {
  bone1Name=hair_c_02
  bone2Name=hair_c_01
  swing1LimitDegrees=45
  swing1Motion=Free
  swing2LimitDegrees=45
  swing2Motion=Free
  twistLimitDegrees=45
  twistMotion=Free
 }
 {
  bone1Name=hair_c_03
  bone2Name=hair_c_02
  swing1LimitDegrees=45
  swing1Motion=Free
  swing2LimitDegrees=45
  swing2Motion=Free
  twistLimitDegrees=45
  twistMotion=Free
 }
 {
  bone1Name=hair_d_01
  bone2Name=head
  swing1LimitDegrees=45
  swing1Motion=Free
  swing2LimitDegrees=45
  swing2Motion=Free
  twistLimitDegrees=45
  twistMotion=Free
 }
 {
  bone1Name=hair_d_02
  bone2Name=hair_d_01
  swing1LimitDegrees=45
  swing1Motion=Free
  swing2LimitDegrees=45
  swing2Motion=Free
  twistLimitDegrees=45
  twistMotion=Free
 }
 {
  bone1Name=hair_d_03
  bone2Name=hair_d_02
  swing1LimitDegrees=45
  swing1Motion=Free
  swing2LimitDegrees=45
  swing2Motion=Free
  twistLimitDegrees=45
  twistMotion=Free
 }
]
